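# Physics asset for the MillionaireTuna backpack cosmetic mesh, covering its rudder bones.
# A tiny capsule sits on base_rudder, and a larger capsule on dyn_rudder is offset and rotated from its bone.
# One constraint joins dyn_rudder to base_rudder with zero twist and swing1 limits and a wide swing2 limit, suggesting a hinge-like rudder swing.
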PhysicsAsset /BRCosmetics/Accessories/FORT_Backpacks/Backpack_M_MED_MillionaireTuna/Meshes/M_MED_MillionaireTuna_Pack_Physics
bodies {
 base_rudder [{shapeType=Capsule, length=0.505, radius=0.505, shapeName=base_rudder_capsule}]
 dyn_rudder [{shapeType=Capsule, length=3, radius=3, shapeName=dyn_rudder_capsule, center=<12 0 0>, rotation=<rot 90 0 0>}]
}
constraints [{bone1Name=dyn_rudder, bone2Name=base_rudder, swing1LimitDegrees=0, swing2LimitDegrees=180, twistLimitDegrees=0}]
