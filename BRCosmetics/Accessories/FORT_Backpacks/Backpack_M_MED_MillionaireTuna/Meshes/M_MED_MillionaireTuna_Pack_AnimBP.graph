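# Anim Blueprint for the MillionaireTuna backpack that adds secondary physics to its fish, spring, tubes and liquid.
# AnimDynamics simulates spring, spine and tubes, then left/right liquid bones, with liquid motion blended by a movement bool.
# Bone offsets react to falling, skydive floating and DBNO, before a RigidBody pass and fixed rudder rotation.
# The EventGraph sets state alphas and gravity via a deprecated override function; the MoveSpring alphas are unused by the AnimGraph.
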AnimBlueprint /BRCosmetics/Accessories/FORT_Backpacks/Backpack_M_MED_MillionaireTuna/Meshes/M_MED_MillionaireTuna_Pack_AnimBP
target_skeleton=&/M_MED_MillionaireTuna_Pack_Skeleton
graphs [
 AnimGraph {nodes [
   n1 = RigidBody!(
    ComponentPose=Local To Component(LocalPose=Use cached pose 'CacheFloating'())
    Alpha
    OverrideWorldGravity:Vector=Get GravityOverride()
    ExternalForce
   ) -> (Pose) {
    bClampLinearTranslationLimitToRefPose=true
    bOverrideWorldGravity=true
    overridePhysicsAsset=&/M_MED_MillionaireTuna_Pack_Physics
    simulationSpace=BaseBoneSpace
    alphaBoolBlend {blendOption=Linear}
    alphaScaleBiasClamp {bInterpResult=true, interpSpeedDecreasing=3, interpSpeedIncreasing=3}
    baseBoneRef {boneName=pack}
    componentAppliedLinearAccClamp=<1500 0 1500>
    componentLinearAccScale {y=0.8}
    componentLinearVelScale {y=0.8}
    overrideWorldGravity {z=-980}
   }
   n2 = "Transform (Modify) Bone - Bone: dyn_rudder"(ComponentPose=@n1, Alpha=1, Translation="0,0,0", Rotation="-90,0,0", Scale="1,1,1") {
    rotationMode=BMM_Additive
    rotationSpace=BCS_BoneSpace
    translationMode=BMM_Additive
    translationSpace=BCS_BoneSpace
    alphaBoolBlend {blendOption=Linear}
    alphaScaleBiasClamp {bInterpResult=true, interpSpeedIncreasing=5}
    boneToModify {boneName=dyn_rudder}
    rotation {pitch=-90}
   }
   Output Pose(Result=Component To Local(ComponentPose=@n2)) {layerGroup=None}
   n3 = Get GravityOverride()
   n4 = Anim Dynamics - Bone: dyn_liquid_back_r!(ComponentPose=Local To Component(LocalPose=Use cached pose 'FishAnimDyn'()), Alpha=0.75, GravityOverride:Vector=@n3, ExternalForce="0,0,0") -> (Pose) {
    angularSpringConstant=50
    bAngularSpring=true
    bGravityOverrideInSimSpace=true
    bLinearSpring=true
    bUseGravityOverride=true
    gravityScale=0
    linearSpringConstant=200
    simulationSpace=BoneRelative
    alphaBoolBlend {blendOption=Linear}
    boundBone {boneName=dyn_liquid_back_r}
    componentLinearAccScale=<0.03 0.05 0.05>
    physicsBodyDefinitions [{
      sphereCollisionRadius=0
      boundBone {boneName=dyn_liquid_back_r}
      boxExtents=<1 1 1>
      constraintSetup {angularLimitsMax=<10 10 10>, angularLimitsMin=<-10 -10 -10>, angularTarget=<1 0 0>, linearAxesMax=<2 0 0>, linearAxesMin=<-2 0 0>}
      localJointOffset=<0.1 0 0>
     }]
    relativeSpaceBone {boneName=base_liquid_master_r}
   }
   n5 = Anim Dynamics - Bone: dyn_liquid_front_r!(ComponentPose=@n4, Alpha=0.75, GravityOverride:Vector=@n3, ExternalForce="0,0,0") -> (Pose) {
    angularSpringConstant=50
    bAngularSpring=true
    bGravityOverrideInSimSpace=true
    bLinearSpring=true
    bUseGravityOverride=true
    gravityScale=0
    linearSpringConstant=200
    simulationSpace=BoneRelative
    alphaBoolBlend {blendOption=Linear}
    boundBone {boneName=dyn_liquid_front_r}
    componentLinearAccScale=<-0.05 0.3 0.3>
    physicsBodyDefinitions [{
      sphereCollisionRadius=0
      boundBone {boneName=dyn_liquid_front_r}
      boxExtents=<1 1 1>
      constraintSetup {angularLimitsMax=<10 10 10>, angularLimitsMin=<-10 -10 -10>, angularTarget=<1 0 0>, linearAxesMax=<2 0 0>, linearAxesMin=<-2 0 0>}
      localJointOffset=<0.1 0 0>
     }]
    relativeSpaceBone {boneName=base_liquid_master_r}
   }
   n6 = Anim Dynamics - Bone: dyn_liquid_out_r!(ComponentPose=@n5, Alpha=0.75, GravityOverride:Vector=@n3, ExternalForce="0,0,0") -> (Pose) {
    angularSpringConstant=50
    bAngularSpring=true
    bGravityOverrideInSimSpace=true
    bLinearSpring=true
    bOverrideLinearDamping=true
    bUseGravityOverride=true
    gravityScale=0
    linearDampingOverride=0.8
    linearSpringConstant=200
    simulationSpace=BoneRelative
    alphaBoolBlend {blendOption=Linear}
    boundBone {boneName=dyn_liquid_out_r}
    componentLinearAccScale=<0.1 0.3 0.3>
    physicsBodyDefinitions [{
      sphereCollisionRadius=0
      boundBone {boneName=dyn_liquid_out_r}
      boxExtents=<1 1 1>
      constraintSetup {angularLimitsMax=<10 10 10>, angularLimitsMin=<-10 -10 -10>, angularTarget=<1 0 0>, linearAxesMax=<5 0 0>, linearAxesMin=<-3 0 0>}
      localJointOffset=<0.1 0 0>
     }]
    relativeSpaceBone {boneName=base_liquid_master_r}
   }
   n7 = Anim Dynamics - Bone: dyn_liquid_inside_r!(ComponentPose=@n6, Alpha=0.75, GravityOverride:Vector=@n3, ExternalForce="0,0,0") -> (Pose) {
    angularSpringConstant=50
    bAngularSpring=true
    bGravityOverrideInSimSpace=true
    bLinearSpring=true
    bOverrideLinearDamping=true
    bUseGravityOverride=true
    gravityScale=0
    linearDampingOverride=0.8
    linearSpringConstant=200
    simulationSpace=BoneRelative
    alphaBoolBlend {blendOption=Linear}
    boundBone {boneName=dyn_liquid_inside_r}
    componentLinearAccScale=<-0.1 0.3 0.3>
    physicsBodyDefinitions [{
      sphereCollisionRadius=0
      boundBone {boneName=dyn_liquid_inside_r}
      boxExtents=<1 1 1>
      constraintSetup {angularLimitsMax=<10 10 10>, angularLimitsMin=<-10 -10 -10>, angularTarget=<1 0 0>, linearAxesMax=<5 0 0>, linearAxesMin=<-3 0 0>}
      localJointOffset=<0.3 0 0>
     }]
    relativeSpaceBone {boneName=base_liquid_master_r}
   }
   n8 = Anim Dynamics - Bone: dyn_liquid_central_r!(ComponentPose=@n7, Alpha=0.75, GravityOverride:Vector=@n3, ExternalForce="0,0,0") -> (Pose) {
    angularSpringConstant=50
    bAngularSpring=true
    bGravityOverrideInSimSpace=true
    bLinearSpring=true
    bUseGravityOverride=true
    gravityScale=0
    linearSpringConstant=200
    simulationSpace=BoneRelative
    alphaBoolBlend {blendOption=Linear}
    boundBone {boneName=dyn_liquid_central_r}
    componentLinearAccScale=<0.1 0.3 0.3>
    physicsBodyDefinitions [{
      sphereCollisionRadius=0
      boundBone {boneName=dyn_liquid_central_r}
      boxExtents=<1 1 1>
      constraintSetup {angularLimitsMax=<10 10 10>, angularLimitsMin=<-10 -10 -10>, angularTarget=<1 0 0>, linearAxesMax=<5 0 0>}
      localJointOffset=<0 0 0>
     }]
    relativeSpaceBone {boneName=base_liquid_master_r}
   }
   n9 = Anim Dynamics - Bone: dyn_liquid_back_l!(ComponentPose=@n8, Alpha=0.75, GravityOverride:Vector=@n3, ExternalForce="0,0,0") -> (Pose) {
    angularSpringConstant=50
    bAngularSpring=true
    bGravityOverrideInSimSpace=true
    bLinearSpring=true
    bUseGravityOverride=true
    gravityScale=0
    linearSpringConstant=200
    simulationSpace=BoneRelative
    alphaBoolBlend {blendOption=Linear}
    boundBone {boneName=dyn_liquid_back_l}
    componentLinearAccScale=<0.03 0.05 0.05>
    physicsBodyDefinitions [{
      sphereCollisionRadius=0
      boundBone {boneName=dyn_liquid_back_l}
      boxExtents=<1 1 1>
      constraintSetup {angularLimitsMax=<10 10 10>, angularLimitsMin=<-10 -10 -10>, angularTarget=<1 0 0>, linearAxesMax=<2 0 0>, linearAxesMin=<-2 0 0>}
      localJointOffset=<0.1 0 0>
     }]
    relativeSpaceBone {boneName=base_liquid_master_l}
   }
   n10 = Anim Dynamics - Bone: dyn_liquid_front_l!(ComponentPose=@n9, Alpha=0.75, GravityOverride:Vector=@n3, ExternalForce="0,0,0") -> (Pose) {
    angularSpringConstant=50
    bAngularSpring=true
    bGravityOverrideInSimSpace=true
    bLinearSpring=true
    bUseGravityOverride=true
    gravityScale=0
    linearSpringConstant=200
    simulationSpace=BoneRelative
    alphaBoolBlend {blendOption=Linear}
    boundBone {boneName=dyn_liquid_front_l}
    componentLinearAccScale=<0.05 0.3 0.3>
    physicsBodyDefinitions [{
      sphereCollisionRadius=0
      boundBone {boneName=dyn_liquid_front_l}
      boxExtents=<1 1 1>
      constraintSetup {angularLimitsMax=<10 10 10>, angularLimitsMin=<-10 -10 -10>, angularTarget=<1 0 0>, linearAxesMax=<2 0 0>, linearAxesMin=<-2 0 0>}
      localJointOffset=<0.1 0 0>
     }]
    relativeSpaceBone {boneName=base_liquid_master_l}
   }
   n11 = Anim Dynamics - Bone: dyn_liquid_out_l!(ComponentPose=@n10, Alpha=0.75, GravityOverride:Vector=@n3, ExternalForce="0,0,0") -> (Pose) {
    angularSpringConstant=50
    bAngularSpring=true
    bGravityOverrideInSimSpace=true
    bLinearSpring=true
    bOverrideLinearDamping=true
    bUseGravityOverride=true
    gravityScale=0
    linearDampingOverride=0.8
    linearSpringConstant=200
    simulationSpace=BoneRelative
    alphaBoolBlend {blendOption=Linear}
    boundBone {boneName=dyn_liquid_out_l}
    componentLinearAccScale=<0.1 0.3 0.3>
    physicsBodyDefinitions [{
      sphereCollisionRadius=0
      boundBone {boneName=dyn_liquid_out_l}
      boxExtents=<1 1 1>
      constraintSetup {angularLimitsMax=<10 10 10>, angularLimitsMin=<-10 -10 -10>, angularTarget=<1 0 0>, linearAxesMax=<3 0 0>, linearAxesMin=<-3 0 0>}
      localJointOffset=<0.1 0 0>
     }]
    relativeSpaceBone {boneName=base_liquid_master_l}
   }
   n12 = Anim Dynamics - Bone: dyn_liquid_inside_l!(ComponentPose=@n11, Alpha=0.75, GravityOverride:Vector=@n3, ExternalForce="0,0,0") -> (Pose) {
    angularSpringConstant=50
    bAngularSpring=true
    bGravityOverrideInSimSpace=true
    bLinearSpring=true
    bOverrideLinearDamping=true
    bUseGravityOverride=true
    gravityScale=0
    linearDampingOverride=0.8
    linearSpringConstant=200
    simulationSpace=BoneRelative
    alphaBoolBlend {blendOption=Linear}
    boundBone {boneName=dyn_liquid_inside_l}
    componentLinearAccScale=<-0.1 0.3 0.3>
    physicsBodyDefinitions [{
      sphereCollisionRadius=0
      boundBone {boneName=dyn_liquid_inside_l}
      boxExtents=<1 1 1>
      constraintSetup {angularLimitsMax=<10 10 10>, angularLimitsMin=<-10 -10 -10>, angularTarget=<1 0 0>, linearAxesMax=<5 0 0>, linearAxesMin=<-3 0 0>}
      localJointOffset=<0.3 0 0>
     }]
    relativeSpaceBone {boneName=base_liquid_master_l}
   }
   n13 = Anim Dynamics - Bone: dyn_liquid_central_l!(ComponentPose=@n12, Alpha=0.75, GravityOverride:Vector=@n3, ExternalForce="0,0,0") -> (Pose) {
    angularSpringConstant=50
    bAngularSpring=true
    bGravityOverrideInSimSpace=true
    bLinearSpring=true
    bUseGravityOverride=true
    gravityScale=0
    linearSpringConstant=200
    simulationSpace=BoneRelative
    alphaBoolBlend {blendOption=Linear}
    boundBone {boneName=dyn_liquid_central_l}
    componentLinearAccScale=<0.1 0.3 0.3>
    physicsBodyDefinitions [{
      sphereCollisionRadius=0
      boundBone {boneName=dyn_liquid_central_l}
      boxExtents=<1 1 1>
      constraintSetup {angularLimitsMax=<10 10 10>, angularLimitsMin=<-10 -10 -10>, angularTarget=<1 0 0>, linearAxesMax=<5 0 0>}
      localJointOffset=<0 0 0>
     }]
    relativeSpaceBone {boneName=base_liquid_master_l}
   }
   Save cached pose 'LiquidAnimDyn'(Pose=Component To Local(ComponentPose=@n13)) {cacheName=LiquidAnimDyn}
   n14 = Use cached pose 'FishAnimDyn'()
   n15 = Use cached pose 'LiquidAnimDyn'()
   n16 = Get Alpha_movement() -> (Alpha_movement:bool)
   n17 = Get Alpha_IsFalling() -> (Alpha_IsFalling:double)
   n18 = "Transform (Modify) Bone - Bone: dyn_eyebrow_l"(
    ComponentPose=Local To Component(LocalPose=Blend Poses by bool(BlendPose_0=@n15, BlendPose_1=@n14, BlendTime_0=0.75, BlendTime_1=1, bActiveValue=@n16) {blendType=Linear})
    Alpha=@n17
    Translation="0,0,2"
    Rotation="0,0,0"
    Scale="1,1,1"
   ) {
    translationMode=BMM_Additive
    translationSpace=BCS_BoneSpace
    alphaBoolBlend {blendOption=Linear}
    alphaScaleBiasClamp {bInterpResult=true, interpSpeedDecreasing=3, interpSpeedIncreasing=3}
    boneToModify {boneName=dyn_eyebrow_l}
    translation {z=2}
   }
   n19 = "Transform (Modify) Bone - Bone: dyn_eyebrow_r"(ComponentPose=@n18, Alpha=@n17, Translation="0,0,2", Rotation="0,0,0", Scale="1,1,1") {
    translationMode=BMM_Additive
    translationSpace=BCS_BoneSpace
    alphaBoolBlend {blendOption=Linear}
    alphaScaleBiasClamp {bInterpResult=true, interpSpeedDecreasing=3, interpSpeedIncreasing=3}
    boneToModify {boneName=dyn_eyebrow_r}
    translation {z=2}
   }
   Save cached pose 'Cache MoveSpring'(Pose=Component To Local(ComponentPose=@n19)) {cacheName=Cache MoveSpring}
   n20 = "Transform (Modify) Bone - Bone: pack"(ComponentPose=Local To Component(LocalPose=Input Pose()), Translation="0,-1,0", Rotation="0,0,0") {
    translationMode=BMM_Additive
    translationSpace=BCS_BoneSpace
    alphaBoolBlend {blendOption=Linear}
    alphaScaleBiasClamp {bInterpResult=true, interpSpeedDecreasing=5, interpSpeedIncreasing=5}
    boneToModify {boneName=pack}
    translation {y=-1}
   }
   n21 = Get GravityOverride()
   n22 = Anim Dynamics - Bone: dyn_spring_02!(ComponentPose=@n20, Alpha=0.7, GravityOverride:Vector=@n21, ExternalForce="0,0,0") -> (Pose) {
    angularDampingOverride=0.99
    angularSpringConstant=10000
    bAngularSpring=true
    bLinearSpring=true
    bOverrideAngularDamping=true
    bUseGravityOverride=true
    gravityScale=0.35
    linearDampingOverride=1
    linearSpringConstant=10000
    simulationSpace=BoneRelative
    alphaBoolBlend {blendOption=Linear}
    boundBone {boneName=dyn_spring_02}
    componentAppliedLinearAccClamp=<3000 3000 3000>
    componentLinearAccScale=<0.3 0.3 0.3>
    componentLinearVelScale=<0.1 0.1 0.1>
    gravityOverride {z=-980}
    physicsBodyDefinitions [{
      boundBone {boneName=dyn_spring_02}
      boxExtents=<15 15 15>
      constraintSetup {angularLimitsMax=<60 60 0>, angularLimitsMin=<-60 -60 -30>, angularTarget=<1 0 0>}
      localJointOffset=<2 0 0>
     }]
    relativeSpaceBone {boneName=root}
   }
   n23 = Anim Dynamics - Bone: dyn_fish_spine_01!(ComponentPose=@n22, Alpha=Get Alpha_IsNotFrontEnd() -> (Alpha_IsNotFrontEnd:double), GravityOverride:Vector=@n21, ExternalForce="0,0,0") -> (Pose) {
    angularDampingOverride=1
    angularSpringConstant=20000
    bAngularSpring=true
    bOverrideAngularDamping=true
    bOverrideLinearDamping=true
    bUseGravityOverride=true
    gravityScale=0.35
    linearDampingOverride=1
    linearSpringConstant=20000
    simulationSpace=BoneRelative
    alphaBoolBlend {blendOption=Linear}
    boundBone {boneName=dyn_fish_spine_01}
    componentAppliedLinearAccClamp=<3000 3000 3000>
    componentLinearAccScale=<0.1 0.1 0.1>
    componentLinearVelScale=<0.1 0.1 0.1>
    gravityOverride {z=-980}
    physicsBodyDefinitions [{
      boundBone {boneName=dyn_fish_spine_01}
      boxExtents=<6 6 6>
      constraintSetup {angularLimitsMax=<30 0 30>, angularLimitsMin=<-30 0 -30>, angularTarget=<1 0 0>}
      localJointOffset=<3.5 0 0>
     }]
    relativeSpaceBone {boneName=root}
   }
   n24 = Anim Dynamics - Bone: dyn_spring_02(ComponentPose=@n23, Alpha=Get Alpha_IsTacticalSprinting() -> (Alpha_IsTacticalSprinting:double), GravityOverride=@n21, ExternalForce="0,0,0") {
    angularDampingOverride=0.99
    angularSpringConstant=10000
    bAngularSpring=true
    bLinearSpring=true
    bOverrideAngularDamping=true
    bUseGravityOverride=true
    gravityScale=0.35
    linearDampingOverride=1
    linearSpringConstant=10000
    simulationSpace=BoneRelative
    alphaBoolBlend {blendOption=Linear}
    boundBone {boneName=dyn_spring_02}
    componentAppliedLinearAccClamp=<3000 3000 3000>
    componentLinearAccScale=<0.33 0.33 0.33>
    componentLinearVelScale=<0.1 0.1 0.1>
    gravityOverride {z=-980}
    physicsBodyDefinitions [{
      boundBone {boneName=dyn_spring_02}
      boxExtents=<15 15 15>
      constraintSetup {angularLimitsMax=<80 80 0>, angularLimitsMin=<-80 -90 -80>, angularTarget=<1 0 0>}
      localJointOffset=<2 0 0>
     }]
    relativeSpaceBone {boneName=root}
   }
   n25 = Anim Dynamics - Bone: dyn_tube_top_l!(ComponentPose=@n24, Alpha=1, GravityOverride:Vector=@n21, ExternalForce="0,0,0") -> (Pose) {
    angularDampingOverride=0.98
    angularSpringConstant=17000
    bAngularSpring=true
    bLinearSpring=true
    bOverrideAngularDamping=true
    bUseGravityOverride=true
    gravityScale=0.35
    linearSpringConstant=17000
    simulationSpace=BoneRelative
    alphaBoolBlend {blendOption=Linear}
    boundBone {boneName=dyn_tube_top_l}
    componentAppliedLinearAccClamp=<5000 5000 5000>
    componentLinearAccScale=<2 2 2>
    componentLinearVelScale=<1 1 1>
    gravityOverride {z=-980}
    physicsBodyDefinitions [{
      boundBone {boneName=dyn_tube_top_l}
      boxExtents=<20 30 30>
      constraintSetup {angularLimitsMax=<0 0 90>, angularLimitsMin=<0 0 -90>, angularTarget=<1 0 0>}
      localJointOffset=<-3 0 0>
     }]
    relativeSpaceBone {boneName=root}
   }
   n26 = Anim Dynamics - Bone: dyn_tube_top_r!(ComponentPose=@n25, Alpha=1, GravityOverride:Vector=@n21, ExternalForce="0,0,0") -> (Pose) {
    angularDampingOverride=0.98
    angularSpringConstant=17000
    bAngularSpring=true
    bLinearSpring=true
    bOverrideAngularDamping=true
    bUseGravityOverride=true
    gravityScale=0.35
    linearSpringConstant=17000
    simulationSpace=BoneRelative
    alphaBoolBlend {blendOption=Linear}
    boundBone {boneName=dyn_tube_top_r}
    componentAppliedLinearAccClamp=<5000 5000 5000>
    componentLinearAccScale=<2 2 2>
    componentLinearVelScale=<1 1 1>
    gravityOverride {z=-980}
    physicsBodyDefinitions [{
      boundBone {boneName=dyn_tube_top_r}
      boxExtents=<20 30 30>
      constraintSetup {angularLimitsMax=<0 0 90>, angularLimitsMin=<0 0 -90>, angularTarget=<1 0 0>}
      localJointOffset=<-3 0 0>
     }]
    relativeSpaceBone {boneName=root}
   }
   Save cached pose 'FishAnimDyn'(Pose=Component To Local(ComponentPose=@n26)) {cacheName=FishAnimDyn}
   n27 = Get Alpha_IsFloating() -> (Alpha_IsFloating:double)
   n28 = "Transform (Modify) Bone - Bone: base_liquid_front_r"(
    ComponentPose=Local To Component(LocalPose=Use cached pose 'Cache MoveSpring'())
    Alpha=@n27
    Translation="7.5,1,0"
    Rotation="0,0,0"
    Scale="1,1,1"
   ) {
    translationMode=BMM_Additive
    translationSpace=BCS_BoneSpace
    alphaBoolBlend {blendOption=Linear}
    alphaScaleBiasClamp {bInterpResult=true, interpSpeedDecreasing=5, interpSpeedIncreasing=5}
    boneToModify {boneName=base_liquid_front_r}
    translation=<7.5 1>
   }
   n29 = "Transform (Modify) Bone - Bone: base_liquid_back_r"(ComponentPose=@n28, Alpha=@n27, Translation="6,0,0", Rotation="0,0,0", Scale="1,1,1") {
    translationMode=BMM_Additive
    translationSpace=BCS_BoneSpace
    alphaBoolBlend {blendOption=Linear}
    alphaScaleBiasClamp {bInterpResult=true, interpSpeedDecreasing=5, interpSpeedIncreasing=5}
    boneToModify {boneName=base_liquid_back_r}
    translation {x=6}
   }
   n30 = "Transform (Modify) Bone - Bone: base_liquid_front_l"(ComponentPose=@n29, Alpha=@n27, Translation="7.5,1,0", Rotation="0,0,0", Scale="1,1,1") {
    translationMode=BMM_Additive
    translationSpace=BCS_BoneSpace
    alphaBoolBlend {blendOption=Linear}
    alphaScaleBiasClamp {bInterpResult=true, interpSpeedDecreasing=5, interpSpeedIncreasing=5}
    boneToModify {boneName=base_liquid_front_l}
    translation=<7.5 1>
   }
   n31 = "Transform (Modify) Bone - Bone: base_liquid_back_l"(ComponentPose=@n30, Alpha=@n27, Translation="6,0,0", Rotation="0,0,0", Scale="1,1,1") {
    translationMode=BMM_Additive
    translationSpace=BCS_BoneSpace
    alphaBoolBlend {blendOption=Linear}
    alphaScaleBiasClamp {bInterpResult=true, interpSpeedDecreasing=5, interpSpeedIncreasing=5}
    boneToModify {boneName=base_liquid_back_l}
    translation {x=6}
   }
   n32 = Get Alpha_IsDBNO() -> (Alpha_IsDBNO:double)
   n33 = "Transform (Modify) Bone - Bone: base_tube_r"(ComponentPose=@n31, Alpha=@n32, Translation="0,8,0", Rotation="0,0,0", Scale="1,1,1") {
    translationMode=BMM_Additive
    translationSpace=BCS_BoneSpace
    alphaBoolBlend {blendOption=Linear}
    alphaScaleBiasClamp {bInterpResult=true, interpSpeedDecreasing=5, interpSpeedIncreasing=5}
    boneToModify {boneName=base_tube_r}
    translation {y=8}
   }
   n34 = "Transform (Modify) Bone - Bone: base_tube_l"(ComponentPose=@n33, Alpha=@n32, Translation="0,8,0", Rotation="0,0,0", Scale="1,1,1") {
    translationMode=BMM_Additive
    translationSpace=BCS_BoneSpace
    alphaBoolBlend {blendOption=Linear}
    alphaScaleBiasClamp {bInterpResult=true, interpSpeedDecreasing=5, interpSpeedIncreasing=5}
    boneToModify {boneName=base_tube_l}
    translation {y=8}
   }
   Save cached pose 'CacheFloating'(Pose=Component To Local(ComponentPose=@n34)) {cacheName=CacheFloating}
  ]}
 EventGraph {nodes [
   Event BlueprintUpdateAnimation()
   > n1 = Sequence!(execute) -> (then_0, then_1, then_2:Exec, then_3:Exec, then_4:Exec, then_5:Exec, then_6:Exec, then_7:Exec, then_8:Exec, then_9:Exec)
   n2 = NOT Boolean(A=Get bIsCrouchSprinting())
   n3 = Get bIsMoving2D()
   n4 = Get bIsCrouching()
   n5 = Branch(execute=@n1.then_4, Condition=OR Boolean(A=AND Boolean!(A=@n3, B=@n4, C:bool=@n2) -> (ReturnValue), B=Get bIsSkyDiving()))
   > Set Alpha_IsMoveLittleSpring(Alpha_IsMoveLittleSpring:double=1) -> (Output_Get:double)
   Set Alpha_IsMoveLittleSpring(execute=@n5.else, Alpha_IsMoveLittleSpring=0)
   n6 = Branch(execute=@n1.then_9, Condition=NOT Boolean(A=Get bIsInFrontEnd()))
   > Set Alpha_IsNotFrontEnd(Alpha_IsNotFrontEnd:double=0.5) -> (Output_Get:double)
   Set Alpha_IsNotFrontEnd(execute=@n6.else, Alpha_IsNotFrontEnd=0)
   n7 = NOT Boolean(A=Get bIsCrouching())
   n8 = NOT Boolean(A=Get bIsSkyDiving())
   n9 = OR Boolean!(A=Get bIsSprinting(), B=Get bIsTacticalSprinting(), C:bool=Get bIsMoving2D()) -> (ReturnValue)
   n10 = Branch(execute=@n1.then_1, Condition=OR Boolean!(A=AND Boolean(A=@n9, B=@n8, C=@n7), B=Get bIsCrouchSprinting()) -> (ReturnValue))
   > Set Alpha_IsMoveSpring(Alpha_IsMoveSpring:double=1) -> (Output_Get:double)
   Set Alpha_IsMoveSpring(execute=@n10.else, Alpha_IsMoveSpring=0)
   n11 = Branch(execute=@n1.then_7, Condition=Get bIsTacticalSprinting())
   > Set Alpha_IsTacticalSprinting(Alpha_IsTacticalSprinting:double=1) -> (Output_Get:double)
   Set Alpha_IsTacticalSprinting(execute=@n11.else, Alpha_IsTacticalSprinting=0)
   n12 = Branch(execute=@n1.then_8, Condition=Get bIsDBNO())
   > Set Alpha_IsDBNO(Alpha_IsDBNO:double=1) -> (Output_Get:double)
   Set Alpha_IsDBNO(execute=@n12.else, Alpha_IsDBNO=0)
   n13 = Get bIsSkyDiving()
   n14 = Branch(execute=@n1.then_3, Condition=AND Boolean!(A=Get bIsFalling(), B=NOT Boolean(A=@n13)) -> (ReturnValue))
   > Set Alpha_IsFalling(Alpha_IsFalling:double=1) -> (Output_Get:double)
   Set Alpha_IsFalling(execute=@n14.else, Alpha_IsFalling=0)
   Set sprintJumping(execute=@n1.then_2, sprintJumping:bool=OR Boolean!(A=Get bIsMoving2D(), B=Get bIsSprinting(), C:bool=Get bIsJumping()) -> (ReturnValue)) -> (Output_Get:bool)
   n15 = PlayerMovement(execute=@n1.then_5, AnimInstance=Self-Reference(), XDotProductCoefficient=100, YDotProductCoefficient=100, ZDotProductCoefficient=100)
   n16 = BreakVector(InVec=MakeVector(X=@n15, Y=@n15.YMovement))
   > Set Alpha_movement(Alpha_movement:bool=OR Boolean!(
     A="float < float"(A=@n16, B=-50)
     B="float > float"(A=@n16, B=50)
     C="float > float"(A=@n16.Y, B=40)
     D:bool="float < float"(A=@n16.Y, B=-40)
     E:bool=Get bIsPlayingMeleeAnim()
     F:bool=Get bIsJumping()
     G:bool=Get bIsPlayingEmote()
     H:bool=Get bIsInFrontEnd()
     I:bool=Get bIsOnGround() -> (bIsOnGround:bool)
    ) -> (ReturnValue)) -> (Output_Get:bool)
   n17 = Gravity Override State Function DEPRECATED(
    execute=@n1.then_0
    CharacterPartAnimInstance=Self-Reference()
    GameTimeInSeconds=GetGameTimeInSeconds()
    GravityOverrideParams=Get Gravity Override Params()
   )
   > Set GravityOverride(GravityOverride=@n17)
   n18 = Branch(execute=@n1.then_6, Condition=Get bIsSkydiveFloating())
   > Set Alpha_IsFloating(Alpha_IsFloating:double=1) -> (Output_Get:double)
   Set Alpha_IsFloating(execute=@n18.else, Alpha_IsFloating=0)
   Get bIsPlayingMeleeAnim()
  ]}
]
variables [
 Alpha_IsDBNO {type=number}
 Alpha_IsFalling {type=number}
 Alpha_IsFloating {type=number}
 Alpha_IsMoveLittleSpring {type=number}
 Alpha_IsMoveSpring {type=number}
 Alpha_IsNotFrontEnd {type=number}
 Alpha_IsTacticalSprinting {type=number}
 Alpha_movement {type=boolean}
 amplitudWaveHairback {type=number}
 Gravity Override Params {
  type=GravityOverrideParamsStruct
  default {
   dBNO {gravityOverride=<-2000 -400 0>}
   emoteOrMelee {gravityOverride=<0 -300 -980>, windFrequency=<0 0 2>}
   onGround_Moving {windAmplitude=<0 500 -500>, windFrequency=<0 10 10>}
   skydive_DivingDown {gravityOverride=<-250 0 -200>, windAmplitude=<0 0 0>, windFrequency=<0 1 1>}
   skydive_FloatingDown {gravityOverride=<-250 0 -200>, windAmplitude=<200 100 100>, windFrequency=<0 1 1>, windOffset=<0 0 0>}
   skydive_GoingUp {gravityOverride=<0 -300 -980>, windFrequency=<1 2 4>}
  }
 }
 GravityOverride {type=Vector}
 littlespring {type=Vector}
 offsetWaveHairBack {type=number}
 rotatorWaveHairBack {type=Rotator}
 speedWaveHairBack {type=number}
 sprintJumping {type=boolean}
]
